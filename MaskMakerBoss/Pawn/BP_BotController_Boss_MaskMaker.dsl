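; Bot controller for the MaskMaker boss that locks target switching once the boss changes targets.
; On BeginPlay it runs parent logic and binds a handler to the perception current-target-changed event.
; When the target changes, the handler tells the MaskMaker's AI random targeter component that it can no longer switch targets.
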(event EventBeginPlay
  (|CallParentFunction)
  (Perception|BindEventtoOnCurrentTargetChangedEvent (delegate Custom|TargetJustChanged)))

(event Custom|TargetJustChanged
  (params (OldTarget :type Actor%20Object%20Reference) (NewTarget :type Actor%20Object%20Reference))
  (Variables|/MaskMakerBoss/Pawn/FortControllerComponent_AIRandomTargeter_MaskMaker.FortControllerComponent_AIRandomTargeter_MaskMaker_C|SetCanSwitchTarget false (Variables|Default|GetFortControllerComponent_AIRandomTargeter_MaskMaker)))

(fn ConstructionScript
  (|CallParentFunction))
