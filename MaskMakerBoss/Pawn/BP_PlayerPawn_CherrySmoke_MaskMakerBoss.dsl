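; Player pawn for the Mask Maker boss, holding its combat state, effects and ability helpers.
; Breaking all four mask weakpoints triggers a stun; accumulated stun damage past a threshold tags it to end.
; A timer eases the rise to floating; chained EQS queries spawn spire shards in widening rings, preferring player-based locations.
; Hotfix values scale with team size, cooldowns also with health, and target changes turn nearby minions hostile.
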(event Custom|RotateTowardPortal
  (bind (_returnvalue_roll _returnvalue_pitch _returnvalue_yaw) (Math|Rotator|BreakRotator (Math|Rotator|FindLookatRotation (Transformation|GetActorLocation) (Transformation|GetActorLocation (Variables|Default|GetPortalTargetPoint)))))
  ;[0.1362 0.07778 0.15 0.5] RotateTowardPortal
  (Transformation|SetActorRotation 0 (Math|Rotator|MakeRotator 0.0 0.0 _returnvalue_yaw) false))

(event Custom|RemoveDmgResistGE
  ;[0.1361 0.08129 0.15 0.5] RemoveDmgResistGE
  (GameplayEffects|RemoveActiveGameplayEffect (Ability|GetAbilitySystemComponent 0) (Variables|Default|GetDmgResistGE) -1))

(event Custom|TargetSwitchNoNewTarget
  (bind _maskmakercontroller (Variables|Default|GetMaskMakerController))
  (bind _fortcontrollercomponent_airandomtargeter_maskmaker (Variables|/MaskMakerBoss/Pawn/BP_BotController_Boss_MaskMaker.BP_BotController_Boss_MaskMaker_C|GetFortControllerComponent_AIRandomTargeter_MaskMaker _maskmakercontroller))
  ;[0.1384 0.0889 0.15 0.5] TargetSwitchNoNewTarget
  (Utilities|FlowControl|RetriggerableDelay (ScalableFloat|GetValueatLevel (Variables|Hotfix|GetTargetSwitchNoNewTargetDelay_Hf)))
  (if (Variables|/MaskMakerBoss/Pawn/FortControllerComponent_AIRandomTargeter_MaskMaker.FortControllerComponent_AIRandomTargeter_MaskMaker_C|GetCanSwitchTarget _fortcontrollercomponent_airandomtargeter_maskmaker)
    (Variables|/MaskMakerBoss/Pawn/FortControllerComponent_AIRandomTargeter_MaskMaker.FortControllerComponent_AIRandomTargeter_MaskMaker_C|SetCanSwitchTarget false _fortcontrollercomponent_airandomtargeter_maskmaker)))

(event Custom|ActivateJustFloatedUpCd
  ;[0.15 0.08671 0.1485 0.5] ActivateJustFloatedUpCd
  (Utilities|FlowControl|Delay (ScalableFloat|GetValueatLevel (Variables|Hotfix|GetJustFloatedUpDelay_Hf)))
  (if (not (Utilities|Enum|Equal(Enum) (Class|FortAthenaAIBotController|GetAlertLevel (Variables|Default|GetMaskMakerController)) "Unaware"))
    (Variables|Default|SetJustFloatedUp false)))

(event EventBeginPlay
  (bind _maskmakercontroller (Variables|Default|GetMaskMakerController))
  (bind _fortathenanpcpatrolling (Variables|/Game/Athena/AI/NPCs/Base/BP_BotController_NPC_Base.BP_BotController_NPC_Base_C|GetFortAthenaNpcPatrolling _maskmakercontroller))
  (bind _returnvalue (Class|FortAthenaNpcPatrollingComponent|GetPatrolPath _fortathenanpcpatrolling :!class /Script/FortniteAI.FortAthenaNpcPatrollingComponent))
  (bind _returnvalue_2 (Class|FortAthenaPatrolPath|GetPatrolPoint _returnvalue 0))
  (bind _mesh (Variables|Character|GetMesh))
  ;[0.06645 0.09698 0.15 0.5] BeginPlay
  (|CallParentFunction)
  (Networking|SwitchHasAuthority
    (:Authority
      (bind (_gamestate _playlist) (Class|FortAsyncActionCurrentPlaylistReady|PlaylistDataReadyAsync)
        (:OnReady
          (Variables|Default|SetMaxTeamSize (Math|Integer|Clamp(Integer) (Variables|/Script/FortniteGame.FortPlaylist|GetMaxTeamSize _playlist) 1 4))
          (bind _asbp_bot_controller_boss_mask_maker (Utilities|Casting|CastToBP_BotController_Boss_MaskMaker (Pawn|GetController)))
          (bind _output_get (Variables|Default|SetMaskMakerController _asbp_bot_controller_boss_mask_maker))
          (AlertLevel|BindEventtoOnAlertLevelChangedEvent _output_get (EventDispatchers|CreateEvent :!event Custom|AlertLevelChangedEvent))
          (Perception|BindEventtoOnCurrentTargetChangedEvent _output_get (EventDispatchers|CreateEvent :!event Custom|MaskMakerTargetChanged))
          (Perception|BindEventtoOnCurrentTargetChangedEvent _output_get (EventDispatchers|CreateEvent :!event Custom|UpdateBossTarget))
          (EventDispatchers|BindEventtoOnDied (delegate Custom|Died))
          (Weapon|SetHolsterWeapon 0 true)
          (CallFunction|InitiateWeakpointsData)
          (CallFunction|InitializePortalTargetPoint)
          (CallFunction|InitializeTeleportAreaBox)
          (for _index (range (Components|SkinnedMesh|GetNumBones _mesh))
            (Utilities|Array|AddUnique (Variables|Default|GetBoneNames) (Components|SkinnedMesh|GetBoneName _mesh _index)))
          (Utilities|IsValid _returnvalue_2
            (:Is%20Valid
              (Variables|Default|SetSpawnLocation (Utilities|Operators|Add (Transformation|GetActorLocation _returnvalue_2) (Math|Vector|MakeVector 0.0 0.0 (Components|Capsule|GetScaledCapsuleHalfHeight (Variables|Character|GetCapsuleComponent))))))
            (:Is%20Not%20Valid
              (Variables|Default|SetSpawnLocation (Transformation|GetActorLocation)))))))))

(event Custom|Died
  (params (DamagedActor :type Actor%20Object%20Reference) (Damage :type Float%20%28single-precision%29) (InstigatedBy :type Controller%20Object%20Reference) (DamageCauser :type Actor%20Object%20Reference) (HitLocation :type Vector) (FHitComponent :type Primitive%20Component%20Object%20Reference) (BoneName :type Name) (Momentum :type Vector))
  ;[0.06645 0.09698 0.15 0.5] BeginPlay
  (Components|Activation|Deactivate (Variables|Default|GetAthenaSpecialActor)))

(event Custom|UpdateBossTarget
  (params (OldTarget :type Actor%20Object%20Reference) (NewTarget :type Actor%20Object%20Reference))
  ;[0.15 0.07009 0.1354 0.5] UpdateBossTarget
  (bind _asfort_pawn (Utilities|Casting|CastToFortPawn NewTarget))
  (Variables|Default|SetCurrentBossTarget _asfort_pawn))

(event Custom|ActivateStunnedState
  (bind _challengemanagerref (Variables|Default|GetChallengeManagerRef))
  ;[0.15 0.07297 0.1443 0.5] ActivateStunnedState
  (Variables|Default|SetIsInStunnedState true)
  (Default|CallBossStunned)
  (Utilities|IsValid _challengemanagerref
    (:Is%20Valid
      (Class|MaskMakerChallengeManagerInterface|BossStunned(Message) _challengemanagerref))))

(event Custom|StartDelayBetweenAbilityCd
  ;[0.15 0.07463 0.1485 0.5] StartDelayBetweenAbilityCd
  (Class|BPFLAICommonFunctions|AddUniqueGameplayTag self "(TagName=\"MaskMakerBoss.Ability.BetweenCooldown\")")
  (Utilities|FlowControl|Delay (Math|Float|Lerp (ScalableFloat|GetValueatLevel (Variables|Hotfix|GetBetweenAbilityMinCd_Hf) (Math|Conversions|ToFloat(Integer) (Variables|Default|GetMaxTeamSize))) (ScalableFloat|GetValueatLevel (Variables|Hotfix|GetBetweenAbilityMaxCd_Hf) (Math|Conversions|ToFloat(Integer) (Variables|Default|GetMaxTeamSize))) (/ (FortPawn|GetHealth) (FortPawn|GetMaxHealth))))
  (Class|BPFLAICommonFunctions|RemoveAllInstancesOfGameplayTag self "(TagName=\"MaskMakerBoss.Ability.BetweenCooldown\")"))

(event Custom|AlertLevelChangedEvent
  (params (BotController :type Fort%20Athena%20AIBot%20Controller%20Object%20Reference) (OldAlertLevel :type EAlertLevel%20Enum :default "Unaware") (NewAlertLevel :type EAlertLevel%20Enum :default "Unaware"))
  ;[0.06374 0.15 0.06244 0.5] AlertLevelChanged
  (switch Utilities|FlowControl|Switch|SwitchonEAlertLevel NewAlertLevel
    (:Unaware
      (Variables|Default|SetDoUnawareIdle true))
    (:Threatened
      (Variables|Default|SetDoUnawareIdle false)
      (CallFunction|StartAbilityCooldowns))))

(event Custom|MaskWeakpointHealthUpdate
  (params (MaskCurrentHealth :type Float%20%28double-precision%29) (MaskRef :type BGA%20Weakpoint%20Actor%20Mask%20Maker%20Object%20Reference))
  (bind _maskweakpoints (Variables|Default|GetMaskWeakpoints))
  (bind _returnvalue (Utilities|Array|FindItem _maskweakpoints MaskRef))
  (bind _challengemanagerref (Variables|Default|GetChallengeManagerRef))
  ;[0.137 0.08027 0.15 0.5] MaskWeakpointHealthUpdate
  (Utilities|Array|SetArrayElem (Variables|Default|GetMaskWeakpointsHealth) _returnvalue MaskCurrentHealth true)
  (Default|CallWeakpointHealthUpdated _returnvalue MaskCurrentHealth)
  (Utilities|IsValid _challengemanagerref
    (:Is%20Valid
      (Class|MaskMakerChallengeManagerInterface|WeakpointHealthUpdated(Message) _challengemanagerref _returnvalue MaskCurrentHealth))
    (:Is%20Not%20Valid
      _))
  (if (CallFunction|IsEveryMaskDisabled)
    (CallFunction|ActivateStunnedState)))

(event Custom|StartRiseToFloating
  (params (RiseAnimMontage :type Anim%20Montage%20Object%20Reference) (RiseAnimLenght :type Float%20%28double-precision%29) (RiseAnimPlayRate :type Float%20%28double-precision%29 :default 1.0) (RiseAnimSectionName :type Name))
  (bind _capsulecomponent (Variables|Character|GetCapsuleComponent))
  (bind _returnvalue (Components|Capsule|GetScaledCapsuleHalfHeight _capsulecomponent))
  (bind _portaltargetpoint (Variables|Default|GetPortalTargetPoint))
  (bind (_returnvalue_x _returnvalue_y _returnvalue_z) (Math|Vector|BreakVector (Transformation|GetActorLocation _portaltargetpoint)))
  (bind _returnvalue_2 (+ _returnvalue_z _returnvalue))
  (bind (_returnvalue_x_2 _returnvalue_y_2) (Math|Vector|BreakVector (Transformation|GetActorLocation)))
  (bind _returnvalue_3 (Math|Vector|MakeVector _returnvalue_x_2 _returnvalue_y_2 _returnvalue_2))
  ;[0.1773 0.3958 0.08066 0.5] StartRiseToFloating
  (Animation|PlayAnimMontage RiseAnimMontage RiseAnimPlayRate RiseAnimSectionName)
  (Variables|Default|SetRiseToFloatingAnimLenght (* RiseAnimLenght RiseAnimPlayRate))
  (Variables|Default|SetGroundLocation _returnvalue_3)
  (bind (_outhit _returnvalue_4) (Collision|LineTraceByChannel _returnvalue_3 (- _returnvalue_3 "0.000000,0.000000,500.000000") "TraceTypeQuery10"))
  (Variables|Default|SetFloatingLocation (Utilities|Operators|Add (Math|Vector|SelectVector (.ImpactPoint (Collision|BreakHitResult _outhit)) (Variables|Default|GetGroundLocation) _returnvalue_4) (Math|Vector|MakeVector 0.0 0.0 (ScalableFloat|GetValueatLevel (Variables|Hotfix|GetFloatingHeight_Hf)))))
  (Pawn|Components|CharacterMovement|SetMovementMode (Variables|Character|GetCharacterMovement) "MOVE_Custom" (Utilities|Enum|LiteralenumEFortCustomMovement "Floating"))
  (Physics|SetEnableGravity (Variables|Character|GetCapsuleComponent) false)
  (Variables|Default|SetRiseToFloatingAlpha 0.0)
  (bind _returnvalue_5 (Utilities|Time|SetTimerbyEvent (delegate Custom|RiseToFloating) (ScalableFloat|GetValueatLevel (Variables|Hotfix|GetRiseToFloatingPeriod_Hf)) true true))
  (Variables|Default|SetRiseToFloatingTimerHandle _returnvalue_5))

(event Custom|RiseToFloating
  ;[0.1773 0.3958 0.08066 0.5] StartRiseToFloating
  (Variables|Default|SetRiseToFloatingAlpha (+ (Variables|Default|GetRiseToFloatingAlpha) (/ (ScalableFloat|GetValueatLevel (Variables|Hotfix|GetRiseToFloatingPeriod_Hf)) (Variables|Default|GetRiseToFloatingAnimLenght))))
  (Variables|Default|SetMaskMakerLocation (Math|Interpolation|Ease "EaseInOut" (Variables|Default|GetRiseToFloatingAlpha) (Variables|Default|GetGroundLocation) (Variables|Default|GetFloatingLocation) 2.0))
  (if (>= (Variables|Default|GetRiseToFloatingAlpha) 1.0)
    (Utilities|Time|ClearandInvalidateTimerbyHandle (Variables|Default|GetRiseToFloatingTimerHandle))
    (Variables|Default|SetIsFloating true)))

(event Custom|PawnPlayUnawareIdleAnim
  (params (AnimMontage :type Anim%20Montage%20Object%20Reference) (SectionName :type Name) (PlayRate :type Boolean))
  ;[0.0422 0.1119 0.15 0.5] PlayAnimMontage
  (Animation|StopAnimMontage 0)
  (Animation|PlayAnimMontage AnimMontage 1.0 SectionName))

(event Custom|SpawnSpireShardsPlayers
  (bind _shardamountperring_hf (Variables|Hotfix|GetShardAmountPerRing_Hf))
  (bind _returnvalue (ScalableFloat|GetValueatLevel _shardamountperring_hf))
  (bind _shardspawnedamount (Variables|Default|GetShardSpawnedAmount))
  (bind _returnvalue_2 (/ _shardspawnedamount _returnvalue))
  (bind _returnvalue_3 (Math|Float|Floor _returnvalue_2))
  (bind _minsharddistancefromboss_hf (Variables|Hotfix|GetMinShardDistanceFromBoss_Hf))
  (bind _returnvalue_4 (ScalableFloat|GetValueatLevel _minsharddistancefromboss_hf))
  (bind _spireshardsdonutdelta_hf (Variables|Hotfix|GetSpireShardsDonutDelta_Hf))
  (bind _returnvalue_5 (ScalableFloat|GetValueatLevel _spireshardsdonutdelta_hf))
  (bind _returnvalue_6 (* _returnvalue_5 _returnvalue_3))
  (bind _returnvalue_7 (+ _returnvalue_4 _returnvalue_6))
  ;[0.1029 0.03864 0.15 0.5] SpawnSpireShards
  (bind _returnvalue_8 (AI|EQS|RunEQSQuery "/MaskMakerBoss/EQS/EQS_MaskMaker_SpireShardsDonutLocations_Players_Param.EQS_MaskMaker_SpireShardsDonutLocations_Players_Param" self "RandomBest5Pct"))
  (AI|EQS|SetNamedParam _returnvalue_8 "Donut.InnerRadius" _returnvalue_7)
  (AI|EQS|SetNamedParam _returnvalue_8 "Donut.OuterRadius" (+ _returnvalue_7 _returnvalue_5))
  (AI|EQS|BindEventtoOnQueryFinished _returnvalue_8 (delegate Custom|SpireShardsPlayersQueryFinished)))

(event Custom|PawnPlayAbilityAnim
  (params (AnimMontage :type Anim%20Montage%20Object%20Reference) (PlayRate :type Float%20%28double-precision%29 :default 1.0) (SectionName :type Name))
  ;[0.0422 0.1119 0.15 0.5] PlayAnimMontage
  (Animation|PlayAnimMontage AnimMontage PlayRate SectionName))

(event EventOnDamageServer
  (params (Damage :type Float%20%28single-precision%29) (DamageTags :type Gameplay%20Tag%20Container%20Structure%20%28by%20ref%29) (Momentum :type Vector :default "0, 0, 0") (HitInfo :type Hit%20Result%20Structure%20%28by%20ref%29) (InstigatedBy :type Controller%20Object%20Reference) (DamageCauser :type Actor%20Object%20Reference) (EffectContext :type Gameplay%20Effect%20Context%20Handle%20Structure))
  ;[0.15 0.03283 0.03512 0.5] OnDamage
  (|CallParentFunction Damage DamageTags Momentum HitInfo InstigatedBy DamageCauser EffectContext)
  (if (!= DamageCauser self)
    (if (Variables|Default|GetIsInStunnedState)
      (bind _output_get (Variables|Default|SetDamageInStunnedState (+ (Variables|Default|GetDamageInStunnedState) Damage)))
      (GameplayCue|ExecuteGameplayCueOnActor(Burst) self (Variables|Default|GetDmgWhileStunned_GCN) (Ability|GameplayCue|MakeGameplayCueParameters 0.0))
      (if (and (not (GameplayTags|HasAnyMatchingGameplayTags (GameplayTags|MakeGameplayTagContainerfromTag "(TagName=\"MaskMakerBoss.EndStunnedState\")"))) (<= (* (* (ScalableFloat|GetValueatLevel (Variables|Hotfix|GetMaskMakerPlaylistMaxHealth_Hf) (Math|Conversions|ToFloat(Integer) (Variables|Default|GetMaxTeamSize))) 1e+02) (ScalableFloat|GetValueatLevel (Variables|Hotfix|GetDmgThresholdPercent_Hf) (Math|Conversions|ToFloat(Integer) (Variables|Default|GetMaxTeamSize)))) _output_get))
        (Class|BPFLAICommonFunctions|AddUniqueGameplayTag self "(TagName=\"MaskMakerBoss.EndStunnedState\")")
        (if (Math|Random|RandomBoolwithWeight (ScalableFloat|GetValueatLevel (Variables|Hotfix|GetReturnFromStunnedSpeechOdds_Hf)))
          (Class|BPFLAICommonFunctions|PlaySpeechBubble self "(TagName=\"Athena.AI.SpeechBubbles.CherrySmoke.MaskMaker.ReturnFromStun\")" true true)))
      (elif (Utilities|Enum|NotEqual(Enum) (Variables|/Script/FortniteAI.FortAthenaAIBotController|GetCurrentAlertLevel (Variables|Default|GetMaskMakerController)) "Threatened")
        (Class|FortAthenaAIBotController|SetAlertLevel (Variables|Default|GetMaskMakerController) "Threatened" "UpdateNextTick" true)))))

(event Custom|SpireShardsPlayersQueryFinished
  (params (QueryInstance :type EQS%20Query%20Instance%20Object%20Reference) (QueryStatus :type EEnvQueryStatus%20Enum :default "Processing"))
  (bind _minsharddistancefromboss_hf (Variables|Hotfix|GetMinShardDistanceFromBoss_Hf))
  (bind _returnvalue (ScalableFloat|GetValueatLevel _minsharddistancefromboss_hf))
  (bind _spireshardsdonutdelta_hf (Variables|Hotfix|GetSpireShardsDonutDelta_Hf))
  (bind _returnvalue_2 (ScalableFloat|GetValueatLevel _spireshardsdonutdelta_hf))
  (bind _shardamountperring_hf (Variables|Hotfix|GetShardAmountPerRing_Hf))
  (bind _returnvalue_3 (ScalableFloat|GetValueatLevel _shardamountperring_hf))
  (bind _shardspawnedamount (Variables|Default|GetShardSpawnedAmount))
  (bind _returnvalue_4 (/ _shardspawnedamount _returnvalue_3))
  (bind _returnvalue_5 (Math|Float|Floor _returnvalue_4))
  (bind _returnvalue_6 (* _returnvalue_2 _returnvalue_5))
  (bind _returnvalue_7 (+ _returnvalue _returnvalue_6))
  ;[0.1029 0.03864 0.15 0.5] SpawnSpireShards
  (bind (_resultlocations _returnvalue_8) (AI|EQS|GetQueryResultsasLocations QueryInstance))
  (if _returnvalue_8
    (Utilities|Array|Add (Variables|Default|GetSavedSpireShardsLocations) (Utilities|Array|Get(acopy) _resultlocations))
    (Variables|Default|SetNextShardSpawnLoc (Utilities|Array|Get(acopy) _resultlocations))
    (Game|SpawnActorfromClass "/MaskMakerBoss/Slag/B_Prop_SpireSlag_MaskMaker.B_Prop_SpireSlag_MaskMaker_C" (Math|Transform|MakeTransform (Variables|Default|GetNextShardSpawnLoc) (Math|Rotator|MakeRotator (Math|Random|RandomFloatinRange (* (ScalableFloat|GetValueatLevel (Variables|Hotfix|GetMaxXYRotation_Hf)) -1.0) (ScalableFloat|GetValueatLevel (Variables|Hotfix|GetMaxXYRotation_Hf))) (Math|Random|RandomFloatinRange (* (ScalableFloat|GetValueatLevel (Variables|Hotfix|GetMaxXYRotation_Hf)) -1.0) (ScalableFloat|GetValueatLevel (Variables|Hotfix|GetMaxXYRotation_Hf))) (Math|Random|RandomFloatinRange 0.0 359.0)) (Math|Conversions|ToVector(Float) (ScalableFloat|GetValueatLevel (Variables|Hotfix|GetShardStartScale_Hf)))) "AlwaysSpawn" "MultiplyWithRoot" 0 0 0 self)
    (Math|Integer|IncrementInt (Variables|Default|GetTotalShardSpawnedAmount)
      (:%20%20
        (bind _result (Math|Integer|IncrementInt (Variables|Default|GetShardSpawnedAmount))
          (:%20%20
            (if (not (>= _result (ScalableFloat|GetValueatLevel (Variables|Hotfix|GetMaxSpireShardsAmount_Hf))))
              (Utilities|FlowControl|Delay (ScalableFloat|GetValueatLevel (Variables|Hotfix|GetDelayBetweenSpireShardsSpawn_Hf)))
              (CallFunction|SpawnSpireShardsPlayers))))))
    (else
      (bind _returnvalue_9 (AI|EQS|RunEQSQuery "/MaskMakerBoss/EQS/EQS_MaskMaker_SpireShardsDonutLocations_Random_Param.EQS_MaskMaker_SpireShardsDonutLocations_Random_Param" self "RandomBest5Pct"))
      (AI|EQS|SetNamedParam _returnvalue_9 "Donut.InnerRadius" _returnvalue_7)
      (AI|EQS|SetNamedParam _returnvalue_9 "Donut.OuterRadius" (+ _returnvalue_7 _returnvalue_2))
      (AI|EQS|BindEventtoOnQueryFinished _returnvalue_9 (delegate Custom|SpireShardsLocationsQueryFinished)))))

(event Custom|SpireShardsLocationsQueryFinished
  (params (QueryInstance :type EQS%20Query%20Instance%20Object%20Reference) (QueryStatus :type EEnvQueryStatus%20Enum :default "Processing"))
  ;[0.1029 0.03864 0.15 0.5] SpawnSpireShards
  (bind (_resultlocations _returnvalue) (AI|EQS|GetQueryResultsasLocations QueryInstance))
  (if _returnvalue
    (Utilities|Array|Add (Variables|Default|GetSavedSpireShardsLocations) (Utilities|Array|Get(acopy) _resultlocations))
    (Variables|Default|SetNextShardSpawnLoc (Utilities|Array|Get(acopy) _resultlocations))
    (Game|SpawnActorfromClass "/MaskMakerBoss/Slag/B_Prop_SpireSlag_MaskMaker.B_Prop_SpireSlag_MaskMaker_C" (Math|Transform|MakeTransform (Variables|Default|GetNextShardSpawnLoc) (Math|Rotator|MakeRotator (Math|Random|RandomFloatinRange (* (ScalableFloat|GetValueatLevel (Variables|Hotfix|GetMaxXYRotation_Hf)) -1.0) (ScalableFloat|GetValueatLevel (Variables|Hotfix|GetMaxXYRotation_Hf))) (Math|Random|RandomFloatinRange (* (ScalableFloat|GetValueatLevel (Variables|Hotfix|GetMaxXYRotation_Hf)) -1.0) (ScalableFloat|GetValueatLevel (Variables|Hotfix|GetMaxXYRotation_Hf))) (Math|Random|RandomFloatinRange 0.0 359.0)) (Math|Conversions|ToVector(Float) (ScalableFloat|GetValueatLevel (Variables|Hotfix|GetShardStartScale_Hf)))) "AlwaysSpawn" "MultiplyWithRoot" 0 0 0 self)
    (Math|Integer|IncrementInt (Variables|Default|GetTotalShardSpawnedAmount)
      (:%20%20
        (bind _result (Math|Integer|IncrementInt (Variables|Default|GetShardSpawnedAmount))
          (:%20%20
            (if (not (>= _result (ScalableFloat|GetValueatLevel (Variables|Hotfix|GetMaxSpireShardsAmount_Hf))))
              (Utilities|FlowControl|Delay (ScalableFloat|GetValueatLevel (Variables|Hotfix|GetDelayBetweenSpireShardsSpawn_Hf)))
              (CallFunction|SpawnSpireShardsPlayers))))))))

(event EventOnDamagePlayEffects
  (params (Damage :type Float%20%28single-precision%29) (DamageTags :type Gameplay%20Tag%20Container%20Structure%20%28by%20ref%29) (Momentum :type Vector :default "0, 0, 0") (HitInfo :type Hit%20Result%20Structure%20%28by%20ref%29) (InstigatedBy :type Fort%20Pawn%20Object%20Reference) (DamageCauser :type Actor%20Object%20Reference) (EffectContext :type Gameplay%20Effect%20Context%20Handle%20Structure))
  (bind (_bblockinghit _binitialoverlap _time _distance _location _impactpoint _normal _impactnormal) (Collision|BreakHitResult HitInfo))
  (bind _returnvalue (Math|Conversions|RotationFromXVector _impactnormal))
  ;[0.6094 0.05311 0.2812 0.5] PlayRicochetImpactFX
  (|CallParentFunction Damage DamageTags Momentum HitInfo InstigatedBy DamageCauser EffectContext)
  (Development|PrintString (Utilities|String|Append "Is in Stunned State = " (Utilities|String|ToString(Boolean) (Variables|Default|GetIsInStunnedState))) false false "(R=1.000000,G=0.000000,B=0.070813,A=1.000000)" 22.0)
  (if (not (Variables|Default|GetIsInStunnedState))
    (bind _ricochetfx (Variables|Default|GetRicochetFX)
      (:then
        (Components|Activation|Activate _ricochetfx true)
        (Transformation|SetWorldLocationAndRotation _ricochetfx _impactpoint _returnvalue false false))
      (:else
        (bind _returnvalue_2 (Effects|Components|ParticleSystem|SpawnFXAtLocation "/MaskMakerBoss/VFX/NS_MaskMaker_RicochetImpact.NS_MaskMaker_RicochetImpact" _impactpoint _returnvalue))
        (Variables|Default|SetRicochetFX _returnvalue_2)))))

(event EventOnDeathPlayEffects
  (params (Damage :type Float%20%28single-precision%29) (DamageTags :type Gameplay%20Tag%20Container%20Structure%20%28by%20ref%29) (Momentum :type Vector :default "0, 0, 0") (HitInfo :type Hit%20Result%20Structure%20%28by%20ref%29) (InstigatedBy :type Fort%20Pawn%20Object%20Reference) (DamageCauser :type Actor%20Object%20Reference) (EffectContext :type Gameplay%20Effect%20Context%20Handle%20Structure))
  ;[1 1 1] Death FX
  (|CallParentFunction Damage DamageTags Momentum HitInfo InstigatedBy DamageCauser EffectContext)
  (Effects|Components|ParticleSystem|SpawnFXAtLocation "/MaskMakerBoss/VFX/NS_MaskMaker_DeathFX.NS_MaskMaker_DeathFX" (Transformation|GetActorLocation)))

(event EventOnDeathServer
  (params (Damage :type Float%20%28single-precision%29) (DamageTags :type Gameplay%20Tag%20Container%20Structure%20%28by%20ref%29) (Momentum :type Vector :default "0, 0, 0") (HitInfo :type Hit%20Result%20Structure%20%28by%20ref%29) (InstigatedBy :type Controller%20Object%20Reference) (DamageCauser :type Actor%20Object%20Reference) (EffectContext :type Gameplay%20Effect%20Context%20Handle%20Structure))
  ;[1 1 1] DeathDialogue
  (|CallParentFunction Damage DamageTags Momentum HitInfo InstigatedBy DamageCauser EffectContext)
  (if (Math|Random|RandomBoolwithWeight (ScalableFloat|GetValueatLevel (Variables|Hotfix|GetEliminatedSpeechOdds_Hf)))
    (Class|BPFLAICommonFunctions|PlaySpeechBubble self "(TagName=\"Athena.AI.SpeechBubbles.CherrySmoke.MaskMaker.Eliminated\")" true true)))

(event Custom|MaskMakerTargetChanged
  (params (OldTarget :type Actor%20Object%20Reference) (NewTarget :type Actor%20Object%20Reference))
  (Utilities|IsValid (Variables|Default|GetCurrentBossTarget)
    (:Is%20Valid
      (switch Utilities|FlowControl|Switch|SwitchonEAlertLevel (Class|FortAthenaAIBotController|GetAlertLevel (Variables|Default|GetMaskMakerController))
        (:Threatened
          (bind _outactors (Actor|GetAllActorsOfClass "/Game/Athena/AI/NPCs/Base/BP_PlayerPawn_NonParticipant.BP_PlayerPawn_NonParticipant_C"))
          (for _array_element _outactors
            (Utilities|IsValid _array_element
              (:Is%20Valid
                (bind _returnvalue (GameplayTags|HasMatchingGameplayTag(Message) _array_element "(TagName=\"AISpawnerData.Type.CherrySmoke.Minion\")"))
                (if (and _returnvalue (< (Math|Vector|Distance(Vector) (Transformation|GetActorLocation _array_element) (Transformation|GetActorLocation)) (ScalableFloat|GetValueatLevel (Variables|Hotfix|GetTargetPropagationDist_Hf))))
                  (Affiliation|ChangeAttitudeBetweenActors _array_element (Variables|Default|GetCurrentBossTarget) "Hostile"))))))))))

(fn ActivateCriticalWeakpoints
  (params (Activate :type Boolean))
  (DamageZones|UpdateBonesatDamageZoneType "Critical" (select Activate (Variables|Default|GetBoneNames) (Utilities|Array|MakeArray "None"))))

(fn DisableAllWeakpoints
  (for _array_element (Variables|Default|GetMaskWeakpoints)
    (Class|BGAWeakpointActorBase|DisableWeakpoint _array_element)
    (Variables|/ActorWeakpointsCore/GameplayActors/BGA_WeakpointActor_Base.BGA_WeakpointActor_Base_C|SetbCollisionEnabled false _array_element))
  (Actor|Tick|SetActorTickEnabled (Variables|Default|GetWeakpointHolder) false))

(fn EnableAllWeakpoints
  (for (_array_element _array_index) (Variables|Default|GetMaskWeakpoints)
    (Class|BGAWeakpointActorBase|EnableWeakpoint _array_element)
    (Variables|/ActorWeakpointsCore/GameplayActors/BGA_WeakpointActor_Base.BGA_WeakpointActor_Base_C|SetbCollisionEnabled true _array_element)
    (Class|BGAWeakpointActorBase|ResetWeakpointHealth _array_element)
    (CallFunction|MaskWeakpointHealthUpdate (ScalableFloat|GetValueatLevel (Variables|Hotfix|GetWeakpointPlaylistMaxHealth_Hf) (Math|Conversions|ToFloat(Integer) (Variables|Default|GetMaxTeamSize))) (Utilities|Array|Get(acopy) (Variables|Default|GetMaskWeakpoints) _array_index)))
  (Actor|Tick|SetActorTickEnabled (Variables|Default|GetWeakpointHolder) true))

(fn GetAdditiveHitReactsMontage
  (return (select (Variables|Default|GetIsInStunnedState) "/Game/Animation/Game/MainPlayer/Montages/General/AdditiveHitReacts.AdditiveHitReacts" 0)))

(fn GetBeamEndLocation
  (return (Variables|Default|GetBeamEndLocation)))

(fn GetChallengerManagerRef
  (return (Variables|Default|GetChallengeManagerRef)))

(fn GetData
  (return (Utilities|Enum|Equal(Enum) (Variables|Default|GetAlertLevel) "Unaware")))

(fn GetDoFastRise
  (return (Variables|Default|GetDoFastRise)))

(fn GetDoUnawareIdle
  (return (Variables|Default|GetDoUnawareIdle)))

(fn GetIsEveryMaskDisabled
  (return (CallFunction|IsEveryMaskDisabled)))

(fn GetIsFloating
  (return (Variables|Default|GetIsFloating)))

(fn GetIsInStunnedState
  (return (Variables|Default|GetIsInStunnedState)))

(fn GetJustFloatedUp
  (return (Variables|Default|GetJustFloatedUp)))

(fn GetJustUsedSpireSlam
  (return (Variables|Default|GetJustUsedSpireSlam)))

(fn GetMaskWeakpoints
  (return (Variables|Default|GetMaskWeakpoints)))

(fn GetMaxTeamSize
  (return (Variables|Default|GetMaxTeamSize)))

(fn GetPortalTargetPoint
  (bind _portaltargetpoint (Variables|Default|GetPortalTargetPoint))
  (Utilities|IsValid _portaltargetpoint
    (:Is%20Valid
      (return _portaltargetpoint))
    (:Is%20Not%20Valid
      (bind _returnvalue (Actor|GetActorOfClass "/MaskMakerBoss/EQS/BP_PortalTargetPoint.BP_PortalTargetPoint_C"))
      (Utilities|IsValid _returnvalue
        (:Is%20Valid
          (bind _output_get (Variables|Default|SetPortalTargetPoint _returnvalue))
          (return _output_get))
        (:Is%20Not%20Valid
          (return 0))))))

(fn GetSavedSpireShardsLocations
  (return (Variables|Default|GetSavedSpireShardsLocations)))

(fn GetSavedTargetedPlayers
  (return (Variables|Default|GetSavedTargetedPlayers)))

(fn GetSpawnLocation
  (return (Variables|Default|GetSpawnLocation)))

(fn GetTargetLastLocation
  (return (Variables|Default|GetTargetLastLocation)))

(fn GetTeleportAreaBox
  (bind _teleportareabox (Variables|Default|GetTeleportAreaBox))
  (Utilities|IsValid _teleportareabox
    (:Is%20Valid
      (return _teleportareabox))
    (:Is%20Not%20Valid
      (bind _returnvalue (Actor|GetActorOfClass "/MaskMakerBoss/EQS/MaskMaker_TeleportArea_Box.MaskMaker_TeleportArea_Box_C"))
      (Utilities|IsValid _returnvalue
        (:Is%20Valid
          (return _returnvalue))
        (:Is%20Not%20Valid
          (return 0))))))

(fn GetTotalShardSpawnedAmount
  (return (Variables|Default|GetTotalShardSpawnedAmount)))

(fn InitializePortalTargetPoint
  (Utilities|IsValid (Variables|Default|GetPortalTargetPoint)
    (:Is%20Not%20Valid
      (bind _returnvalue (Actor|GetActorOfClass "/MaskMakerBoss/EQS/BP_PortalTargetPoint.BP_PortalTargetPoint_C"))
      (Utilities|IsValid _returnvalue
        (:Is%20Valid
          (Variables|Default|SetPortalTargetPoint _returnvalue))))))

(fn InitializeTeleportAreaBox
  (Utilities|IsValid (Variables|Default|GetTeleportAreaBox)
    (:Is%20Not%20Valid
      (bind _returnvalue (Actor|GetActorOfClass "/MaskMakerBoss/EQS/MaskMaker_TeleportArea_Box.MaskMaker_TeleportArea_Box_C"))
      (Utilities|IsValid _returnvalue
        (:Is%20Valid
          (Variables|Default|SetTeleportAreaBox _returnvalue))))))

(fn InitiateWeakpointsData
  (if (not (Variables|Default|GetWeakpointsInitialized))
    (bind _returnvalue (Game|SpawnActorfromClass "/MaskMakerBoss/Weakpoint/BP_MaskMakerWeakpointHolder.BP_MaskMakerWeakpointHolder_C" (Math|Transform|MakeTransform (Transformation|GetActorLocation)) "Undefined" "MultiplyWithRoot" 0 self))
    (Variables|Default|SetWeakpointHolder _returnvalue)
    (Variables|Default|SetMaskWeakpoints (Utilities|Array|MakeArray (Utilities|Casting|CastToBGA_WeakpointActor_MaskMaker (Variables|/Script/Engine.ChildActorComponent|GetChildActor (Variables|/MaskMakerBoss/Weakpoint/BP_MaskMakerWeakpointHolder.BP_MaskMakerWeakpointHolder_C|GetMaskWeakpoint1 (Variables|Default|GetWeakpointHolder)))) (Utilities|Casting|CastToBGA_WeakpointActor_MaskMaker (Variables|/Script/Engine.ChildActorComponent|GetChildActor (Variables|/MaskMakerBoss/Weakpoint/BP_MaskMakerWeakpointHolder.BP_MaskMakerWeakpointHolder_C|GetMaskWeakpoint2 (Variables|Default|GetWeakpointHolder)))) (Utilities|Casting|CastToBGA_WeakpointActor_MaskMaker (Variables|/Script/Engine.ChildActorComponent|GetChildActor (Variables|/MaskMakerBoss/Weakpoint/BP_MaskMakerWeakpointHolder.BP_MaskMakerWeakpointHolder_C|GetMaskWeakpoint3 (Variables|Default|GetWeakpointHolder)))) (Utilities|Casting|CastToBGA_WeakpointActor_MaskMaker (Variables|/Script/Engine.ChildActorComponent|GetChildActor (Variables|/MaskMakerBoss/Weakpoint/BP_MaskMakerWeakpointHolder.BP_MaskMakerWeakpointHolder_C|GetMaskWeakpoint4 (Variables|Default|GetWeakpointHolder))))))
    (CallFunction|SetWeakpointsMeshes)
    (Variables|Default|SetWeakpointsInitialized true))
  (for _index (range 4)
    (CallFunction|MaskWeakpointHealthUpdate (ScalableFloat|GetValueatLevel (Variables|Hotfix|GetWeakpointPlaylistMaxHealth_Hf) (Math|Conversions|ToFloat(Integer) (Variables|Default|GetMaxTeamSize))) (Utilities|Array|Get(acopy) (Variables|Default|GetMaskWeakpoints) _index))))

(fn IsEveryMaskDisabled
  (if (> (Utilities|Array|Length (Variables|Default|GetMaskWeakpointsHealth)) 0)
    (bind (_array_element _array_index) (Utilities|Array|ForEachLoopwithBreak (Variables|Default|GetMaskWeakpointsHealth))
      (:LoopBody
        (bind _output_get (LocalVariables|SetOneMaskIsEnabled (> _array_element 0)))
        (if (and _output_get (Variables|/ActorWeakpointsCore/GameplayActors/BGA_WeakpointActor_Base.BGA_WeakpointActor_Base_C|GetbIsEnabled (Utilities|Array|Get(acopy) (Variables|Default|GetMaskWeakpoints) _array_index)))
          _))
      (:Completed
        _))
    (if (LocalVariables|GetOneMaskIsEnabled)
      (return false)
      (else
        (return true)))
    (else
      (return false))))

(fn OnRep_DoUnawareIdle
  (Networking|CanExecuteCosmeticEvents
    (:True
      (Effects|Components|ParticleSystem|SetBoolParameter (Variables|Default|GetNS_MaskMaker_IdleLoop) "IsUnawareIdle" (Variables|Default|GetDoUnawareIdle)))))

(fn OnRep_IsInStunnedState
  (bind _returnvalue (Ability|GameplayCue|MakeGameplayCueParameters 0.0))
  (bind _self self)
  (bind _ns_maskmaker_idleloop (Variables|Default|GetNS_MaskMaker_IdleLoop))
  (bind _isinstunnedstate (Variables|Default|GetIsInStunnedState))
  (bind _gc_stunned (Variables|Default|GetGC_Stunned))
  (Utilities|FlowControl|Sequence
    (:then_0
      (Networking|CanExecuteCosmeticEvents
        (:True
          (if _isinstunnedstate
            (Components|Activation|Deactivate _ns_maskmaker_idleloop)
            (else
              (Components|Activation|Activate _ns_maskmaker_idleloop false))))))
    (:then_1
      (if _isinstunnedstate
        (GameplayCue|AddGameplayCueOnActor(Looping) _self _gc_stunned _returnvalue)
        (else
          (GameplayCue|RemoveGameplayCueOnActor(Looping) _self _gc_stunned _returnvalue))))))

(fn OnRep_MaskMakerLocation
  (Transformation|SetActorLocation (Variables|Default|GetMaskMakerLocation)))

(fn SetWeakpointsMeshes
  (bind (_outitem _outindex) (Utilities|Array|RandomArrayItem (LocalVariables|GetMaskMeshArrayLocal)))
  (LocalVariables|SetMaskMeshArrayLocal (Utilities|Array|MakeArray "/MaskMakerBoss/Assets/SolidMist_Mask1/Meshes/SolidMist_Mask1.SolidMist_Mask1" "/MaskMakerBoss/Assets/SolidMist_Mask2/Meshes/SolidMist_Mask2.SolidMist_Mask2" "/MaskMakerBoss/Assets/SolidMist_Mask3/Meshes/SolidMist_Mask3.SolidMist_Mask3" "/MaskMakerBoss/Assets/SolidMist_Mask4/Meshes/SolidMist_Mask4.SolidMist_Mask4"))
  (for _array_element (Variables|Default|GetMaskWeakpoints)
    (LocalVariables|SetChosenIndex _outindex)
    (Class|BGAWeakpointActorMaskMaker|SetMeshfromSoftRef _array_element (Utilities|Array|Get(acopy) (LocalVariables|GetMaskMeshArrayLocal) (LocalVariables|GetChosenIndex)))
    (Utilities|Array|RemoveIndex (LocalVariables|GetMaskMeshArrayLocal) (LocalVariables|GetChosenIndex))))

(fn StartAbilityCooldowns
  (Utilities|FlowControl|Sequence
    (:then_0
      (bind _returnvalue (Spec|MakeSpecHandle(ByClass) "/MaskMakerBoss/GameplayEffect/GE_MaskMaker_BeamAttack_Cooldown.GE_MaskMaker_BeamAttack_Cooldown_C" self self))
      (bind _returnvalue_2 (Ability|GameplayEffect|AssignTagSetbyCallerMagnitude _returnvalue "(TagName=\"SetByCaller.Duration\")" (Math|Float|Lerp (ScalableFloat|GetValueatLevel (Variables|Hotfix|GetBeamAttackMinCd_Hf) (Math|Conversions|ToFloat(Integer) (Variables|Default|GetMaxTeamSize))) (ScalableFloat|GetValueatLevel (Variables|Hotfix|GetBeamAttackMaxCd_Hf) (Math|Conversions|ToFloat(Integer) (Variables|Default|GetMaxTeamSize))) (/ (FortPawn|GetHealth) (FortPawn|GetMaxHealth)))))
      (Environment|ApplyGlobalEnvironmentGameplayEffectSpectoActor self _returnvalue_2))
    (:then_1
      (bind _returnvalue_3 (Spec|MakeSpecHandle(ByClass) "/MaskMakerBoss/GameplayEffect/GE_MaskMaker_SpireShard_Cooldown.GE_MaskMaker_SpireShard_Cooldown_C" self self))
      (bind _returnvalue_4 (Ability|GameplayEffect|AssignTagSetbyCallerMagnitude _returnvalue_3 "(TagName=\"SetByCaller.Duration\")" (Math|Float|Lerp (ScalableFloat|GetValueatLevel (Variables|Hotfix|GetSpireShardsMinCd_Hf) (Math|Conversions|ToFloat(Integer) (Variables|Default|GetMaxTeamSize))) (ScalableFloat|GetValueatLevel (Variables|Hotfix|GetSpireShardsMaxCd_Hf) (Math|Conversions|ToFloat(Integer) (Variables|Default|GetMaxTeamSize))) (/ (FortPawn|GetHealth) (FortPawn|GetMaxHealth)))))
      (Environment|ApplyGlobalEnvironmentGameplayEffectSpectoActor self _returnvalue_4))
    (:then_2
      (bind _returnvalue_5 (Spec|MakeSpecHandle(ByClass) "/MaskMakerBoss/GameplayEffect/GE_MaskMaker_TeleportAOE_Cooldown.GE_MaskMaker_TeleportAOE_Cooldown_C" self self))
      (bind _returnvalue_6 (Ability|GameplayEffect|AssignTagSetbyCallerMagnitude _returnvalue_5 "(TagName=\"SetByCaller.Duration\")" (Math|Float|Lerp (ScalableFloat|GetValueatLevel (Variables|Hotfix|GetTeleportAOEMinCd_Hf) (Math|Conversions|ToFloat(Integer) (Variables|Default|GetMaxTeamSize))) (ScalableFloat|GetValueatLevel (Variables|Hotfix|GetTeleportAOEMaxCd_Hf) (Math|Conversions|ToFloat(Integer) (Variables|Default|GetMaxTeamSize))) (/ (FortPawn|GetHealth) (FortPawn|GetMaxHealth)))))
      (Environment|ApplyGlobalEnvironmentGameplayEffectSpectoActor self _returnvalue_6))))

(fn TrySwitchAggroTarget
  (Variables|/MaskMakerBoss/Pawn/FortControllerComponent_AIRandomTargeter_MaskMaker.FortControllerComponent_AIRandomTargeter_MaskMaker_C|SetCanSwitchTarget true (Variables|/MaskMakerBoss/Pawn/BP_BotController_Boss_MaskMaker.BP_BotController_Boss_MaskMaker_C|GetFortControllerComponent_AIRandomTargeter_MaskMaker (Variables|Default|GetMaskMakerController)))
  (CallFunction|TargetSwitchNoNewTarget))

(fn ConstructionScript
  (|CallParentFunction))
